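# Conversation database for the CritterManiac NPC, giving it an undead-themed greeting before handing off to shared NPC dialogue.
# The entry tag leads to one speech node with a default line and prioritized contextual variants.
# Variants react to time since the last match, a first-ever conversation, held bars, or wearing a specific Phantom outfit.
# The speech node then branches through link nodes to the NPC library Quests and BasicServices conversations.
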ConversationDatabase /NPCLibrary/NPCs/CritterManiac/CritterManiac/NPCLibrary_Conversation_CritterManiac
entry_points [{tag=Conversation.Entry.CritterManiac, destinations [#1]}]
nodes [
 ConversationEntryPointNode_7 {
  id=#1
  class=ConversationEntryPointNode
  outputs [#2]
  properties {entryTag {tagName=Conversation.Entry.CritterManiac}}
 }
 Speech_C_4 {
  id=#2
  class=Speech_C
  outputs [#3, #4]
  properties {
   nodeName=""
   generalConfig {
    defaultMessage="uuUUuuUuUugh… What is this strange new world?"
    contextualMessages [
     {
      message="Has your eternal slumber been disturbed, too?"
      contextRequirements [{requirement {class=FortControllerRequirement_CheckTimeSinceLastPlayedMatch}}]
      priority {curve {rowName=Tandem.Speech.Priority.WhileSincePlayed}}
     }
     {
      message=Who awakens me from my bed of death?
      contextRequirements [{requirement {class=FortControllerRequirement_IsFirstEverConversationWithNPC}}]
      priority {curve {rowName=Tandem.Speech.Priority.FirstEverInteraction}}
     }
     {
      message="You are overladen with treasure, traveler. Let me help you carry the load…"
      contextRequirements [{requirement {class=FortControllerRequirement_HasBars}}]
      priority {curve {rowName=Tandem.Speech.Priority.HighTotalBars}}
     }
     {
      message="Wrath! Have you seen my beloved in the Underworld? Take me to her, PLEASE!"
      contextRequirements [{requirement {
         class=FortControllerRequirement_HasCID
         properties {allowedCIds [/BRCosmetics/Athena/Items/Cosmetics/Characters/CID_599_Athena_Commando_M_Phantom]}
        }}]
      priority {curve {rowName=Tandem.Speech.Priority.OutfitReaction}}
     }
    ]
   }
  }
 }
 ConversationLinkNode_0 {id=#4, class=ConversationLinkNode, properties {remoteEntryTag {tagName=Conversation.Entry.BasicServices}}}
 ConversationLinkNode_1 {id=#3, class=ConversationLinkNode, properties {remoteEntryTag {tagName=Conversation.Entry.NPCLibrary.Quests}}}
]
